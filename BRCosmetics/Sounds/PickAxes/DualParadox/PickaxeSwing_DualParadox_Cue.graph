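# Sound cue for the DualParadox pickaxe swing, mixing close and medium-distance layers with separate attenuation and modulation.
# The close layer switches on SwingIdx, with most inputs sharing one random pool of swing waves and the last input using a random alternate-swing pool.
# The medium layer ignores the index and randomly plays shared Fire Axe distant swing waves.
# Melee sound class, concurrency and a cosmetic pickaxe property sheet govern playback.
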SoundCue /BRCosmetics/Sounds/PickAxes/DualParadox/PickaxeSwing_DualParadox_Cue
root=Mixer_0
nodes [
 {id=Attenuation_0, children [Modulator_0], properties {attenuationSettings=/Game/Sounds/Attenuation/Fort_Weapon_Pickaxe_Close_Att}}
 {id=Attenuation_1, children [Modulator_1], properties {attenuationSettings=/Game/Sounds/Attenuation/Fort_Weapon_Pickaxe_Medium_Att}}
 {id=Mixer_0, children [Attenuation_0, Attenuation_1], properties {inputVolume [1,1]}}
 {id=Modulator_0, children [Switch_0]}
 {id=Modulator_1, children [Random_1]}
 {id=Random_0, children [WavePlayer_0,WavePlayer_1,WavePlayer_2,WavePlayer_3,WavePlayer_4,WavePlayer_5], properties {weights [1,1,1,1,1,1]}}
 {
  id=Random_1
  children [WavePlayer_6, WavePlayer_7, WavePlayer_8, WavePlayer_9, WavePlayer_10, WavePlayer_11]
  properties {weights [1,1,1,1,1,1]}
 }
 {id=Random_2, children [WavePlayer_12, WavePlayer_13, WavePlayer_14], properties {weights [1,1,1]}}
 {id=Switch_0, children [Random_0,Random_0,Random_0,Random_0,Random_2], properties {intParameterName=SwingIdx}}
 {id=WavePlayer_0, properties {soundWaveAssetPtr=&/Pickaxe_DualParadox_Swing_01}}
 {id=WavePlayer_1, properties {soundWaveAssetPtr=&/Pickaxe_DualParadox_Swing_02}}
 {id=WavePlayer_10, properties {soundWaveAssetPtr=/Game/Athena/Sounds/Weapons/Fire_Axe_Swing_Distant_05}}
 {id=WavePlayer_11, properties {soundWaveAssetPtr=/Game/Athena/Sounds/Weapons/Fire_Axe_Swing_Distant_06}}
 {id=WavePlayer_12, properties {soundWaveAssetPtr=&/Pickaxe_DualParadox_SwingAlt_03}}
 {id=WavePlayer_13, properties {soundWaveAssetPtr=&/Pickaxe_DualParadox_SwingAlt_01}}
 {id=WavePlayer_14, properties {soundWaveAssetPtr=&/Pickaxe_DualParadox_SwingAlt_02}}
 {id=WavePlayer_2, properties {soundWaveAssetPtr=&/Pickaxe_DualParadox_Swing_03}}
 {id=WavePlayer_3, properties {soundWaveAssetPtr=&/Pickaxe_DualParadox_Swing_04}}
 {id=WavePlayer_4, properties {soundWaveAssetPtr=&/Pickaxe_DualParadox_Swing_05}}
 {id=WavePlayer_5, properties {soundWaveAssetPtr=&/Pickaxe_DualParadox_Swing_06}}
 {id=WavePlayer_6, properties {soundWaveAssetPtr=/Game/Athena/Sounds/Weapons/Fire_Axe_Swing_Distant_01}}
 {id=WavePlayer_7, properties {soundWaveAssetPtr=/Game/Athena/Sounds/Weapons/Fire_Axe_Swing_Distant_02}}
 {id=WavePlayer_8, properties {soundWaveAssetPtr=/Game/Athena/Sounds/Weapons/Fire_Axe_Swing_Distant_03}}
 {id=WavePlayer_9, properties {soundWaveAssetPtr=/Game/Athena/Sounds/Weapons/Fire_Axe_Swing_Distant_04}}
]
properties {
 audioPropertiesSheet=/BRCosmetics/Sounds/Settings/PropertySheets/Pickaxe/APS_Cosmetic_Pickaxe_Swing
 maxDistance=3000
 soundClassObject=&Fort_WeaponMelee_PSM
 volumeMultiplier=0.2
 concurrencySet [&Player_Melee_Weapon_Conc]
}
